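# Physics asset for the EffigyPin biped cosmetic companion mesh, centred on two dynamic ear chains.
# Each ear is three capsule bodies chained from the head, with limited swing and very tight twist at every joint.
# The head, upper spine and both upper arms also have capsules, with the head and arms joined to spine_05 by free angular constraints, likely as collision shapes for the ears.
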PhysicsAsset /CosmeticCompanions/Assets/Biped/EffigyPin/Meshes/EffigyPin_Biped_Physics
bodies {
 dyn_ear_l_01 [{shapeType=Capsule, length=11.4233, radius=5.8538, shapeName=dyn_ear_l_01_capsule, center=<2.825 -2.2658 3.5353>, rotation=<rot 90 0 0>}]
 dyn_ear_l_02 [{shapeType=Capsule, length=9.8771, radius=7.4, shapeName=dyn_ear_l_02_capsule, center=<2.825 -2.2658 3.5353>, rotation=<rot 90 0 0>}]
 dyn_ear_l_03 [{shapeType=Capsule, length=9.8771, radius=7.4, shapeName=dyn_ear_l_03_capsule, center=<2.825 -2.2658 3.5353>, rotation=<rot 90 0 0>}]
 dyn_ear_r_01 [{
   shapeType=Capsule
   length=9.8771
   radius=7.4
   shapeName=dyn_ear_r_01_capsule
   center=<2.825 -2.2658 -4.95>
   rotation=<rot 85.0079 0.0003 0.0003>
  }]
 dyn_ear_r_02 [{shapeType=Capsule, length=9.8771, radius=7.4, shapeName=dyn_ear_r_02_capsule, center=<2.825 -2.2658 -3.4057>, rotation=<rot 90 0 0>}]
 dyn_ear_r_03 [{shapeType=Capsule, length=9.8771, radius=7.4, shapeName=dyn_ear_r_03_capsule, center=<2.825 -2.2658 -3.4057>, rotation=<rot 90 0 0>}]
 head [{shapeType=Capsule, length=0, radius=15.5321, shapeName=head_capsule, center=<8.6288 3.7075 0>, rotation=<rot 90 0 0>}]
 spine_05 [{shapeType=Capsule, length=9.231, radius=7.3901, shapeName=spine_05_capsule, center=<3.454 0.765 -0.2272>, rotation=<rot -5.7372 -90 -1.04>}]
 upperarm_l [{
   shapeType=Capsule
   length=13.8201
   radius=5.2383
   shapeName=upperarm_l_capsule
   center=<9.8192 2.2386 -0.7257>
   rotation=<rot 0 -70.3878 -81.1973>
  }]
 upperarm_r [{
   shapeType=Capsule
   length=13.311
   radius=4.8904
   shapeName=upperarm_r_capsule
   center=<-10.3548 -2.2242 0.6032>
   rotation=<rot 0 -68.5936 -82.8694>
  }]
}
constraints [
 {bone1Name=dyn_ear_r_02, bone2Name=dyn_ear_r_01, swing1LimitDegrees=30, swing2LimitDegrees=30, twistLimitDegrees=2}
 {bone1Name=dyn_ear_r_03, bone2Name=dyn_ear_r_02, swing1LimitDegrees=30, swing2LimitDegrees=30, twistLimitDegrees=2}
 {bone1Name=dyn_ear_l_02, bone2Name=dyn_ear_l_01, swing1LimitDegrees=30, swing2LimitDegrees=30, twistLimitDegrees=2}
 {bone1Name=dyn_ear_l_03, bone2Name=dyn_ear_l_02, swing1LimitDegrees=30, swing2LimitDegrees=30, twistLimitDegrees=2}
 {bone1Name=dyn_ear_r_01, bone2Name=head, swing1LimitDegrees=30, swing2LimitDegrees=30, twistLimitDegrees=2}
 {bone1Name=dyn_ear_l_01, bone2Name=head, swing1LimitDegrees=30, swing2LimitDegrees=30, twistLimitDegrees=2}
 {
  bone1Name=head
  bone2Name=spine_05
  swing1LimitDegrees=45
  swing1Motion=Free
  swing2LimitDegrees=45
  swing2Motion=Free
  twistLimitDegrees=45
  twistMotion=Free
 }
 {
  bone1Name=upperarm_l
  bone2Name=spine_05
  swing1LimitDegrees=45
  swing1Motion=Free
  swing2LimitDegrees=45
  swing2Motion=Free
  twistLimitDegrees=45
  twistMotion=Free
 }
 {
  bone1Name=upperarm_r
  bone2Name=spine_05
  swing1LimitDegrees=45
  swing1Motion=Free
  swing2LimitDegrees=45
  swing2Motion=Free
  twistLimitDegrees=45
  twistMotion=Free
 }
]
