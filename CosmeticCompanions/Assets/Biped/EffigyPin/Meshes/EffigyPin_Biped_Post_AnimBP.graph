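# Post-process AnimBP for the EffigyPin biped cosmetic companion that adds simulated physics motion to the incoming pose.
# The input pose is cached, converted to component space, run through a RigidBody node using EffigyPin_Biped_Physics, then converted back for output.
# The simulation is based on the head bone, with overridden world gravity and clamped component acceleration.
# A PoseAsset blend node, defaulting to EdibleCable's pose asset, is unconnected; the event graph is empty.
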AnimBlueprint /CosmeticCompanions/Assets/Biped/EffigyPin/Meshes/EffigyPin_Biped_Post_AnimBP
target_skeleton=&/EffigyPin_Biped_Skeleton
graphs [
 AnimGraph {nodes [
   Pose!(PoseAsset:Pose Asset Object Reference=Get PoseAsset() -> (PoseAsset:Pose Asset Object Reference), SourcePose:Pose Link Structure) -> (Pose:Pose Link Structure) {is=title AnimGraphNode_PoseBlendNode, bIgnorePoseAssetSourceAnimationCurves=true, blendOption=Linear}
   n1 = RigidBody(ComponentPose=Local To Component(LocalPose=Use cached pose 'CRCache_Input'())) {
    bOverrideWorldGravity=true
    overridePhysicsAsset=&/EffigyPin_Biped_Physics
    baseBoneRef {boneName=head}
    componentAppliedLinearAccClamp=<3000 3000 3000>
    overrideWorldGravity {z=-980}
   }
   Output Pose(Result=Component To Local(ComponentPose=@n1))
   Save cached pose 'CRCache_Input'(Pose=Input Pose()) {cacheName=CRCache_Input}
  ]}
 EventGraph {links [], nodes []}
]
variables [PoseAsset {type=PoseAsset, default=/CosmeticCompanions/Assets/Biped/EdibleCable/Meshes/EdibleCable_Biped_PoseAsset}]
